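# Report_DM_TextureSize is an Editor Utility Blueprint in the DelMarTools asset review tools, filed under texture reports and named for a texture size report.
# Its event graph is empty and it declares no variables, so the file shows no report logic of its own.
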EditorUtilityBlueprint /DelMarTools/EditorUtilities/AssetReviewTools/Reports/Texture/Report_DM_TextureSize
graphs [EventGraph {links [], nodes []}]
variables []
